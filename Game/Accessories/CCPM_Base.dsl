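; Base custom character part modifier for cosmetic accessories that toggles child particle and mesh visibility as the player's state changes.
; Construction tags child components with the object's display name for visibility toggling.
; On pawn or part change it calls the Attachment hook and, outside the front end, binds res-in/out, hiding-prop and movement events.
; Skydiving and vehicles optionally hide particles, with movement updates detecting vehicle transitions through cross-resetting DoOnce gates.
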(event Collision|EventActorBeginOverlap
  (params (OtherActor :type Actor%20Object%20Reference)))

(event EventTick
  (params (DeltaSeconds :type Float%20%28single-precision%29)))

(event EventOnPawnAndPartChanged
  (bind _outputdelegate (EventDispatchers|CreateEvent :!event Close))
  (bind _outputdelegate_2 (EventDispatchers|CreateEvent :!event Open))
  (|CallParentFunction)
  (Variables|Private|SetCharacter%20Part%20Type (CustomCharacter|GetSkeletalMeshForPartOnPlayerPawn))
  (if (CustomCosmeticModifier|IsinFrontEnd)
    ;[1 1 1] Enable Particle and Mesh visibility
    (CustomCharacter|SetChildParticleComponentsVisible true)
    (CustomCharacter|SetChildMeshComponentsHidden false)
    (CallFunction|Attachment)
    (else
      (Utilities|IsSkydiving (Game|GetPlayerPawn 0)
        (:true
          (if (Variables|CustomCharacterPartModifierBase|GetDisable%20Particles%20during%20Skydive)
            ;[1 1 1 1 50] This will allow players to trade parts. This is future thinking and isnt enabled yet. This does create res in and res out events that toggle hiding of components.
            (CustomCharacter|SetChildParticleComponentsVisible false)))
        (:false
          _))
      ;;[1 1 1] Bind event to player movement. This will check if we are in a vehicle.
      (if (Variables|CustomCharacterPartModifierBase|GetDisable%20Particles%20in%20Vehicle)
        (bind _asplayer_pawn_athena (Utilities|Casting|CastToPlayerPawn_Athena (CustomCharacter|GetAssociatedPlayerPawn))
          (:then
            (Character|BindEventtoOnCharacterMovementUpdated _asplayer_pawn_athena (EventDispatchers|CreateEvent :!event PlayerMovementTick)))
          (:CastFailed
            _)))
      (CallFunction|Attachment) ; Just to the right of this next block.
      (if (and (Utilities|IsValid (Variables|Private|GetLast%20Pawn) :!class /Script/Engine.KismetSystemLibrary) (Utilities|NotEqual(Object) (CustomCharacter|GetAssociatedPlayerPawn) (Variables|Private|GetLast%20Pawn)))
        (Utilities|FlowControl|Sequence
          (:then_0
            (Default|UnbindEventfromResIn (Variables|Private|GetLast%20Pawn) _outputdelegate_2)
            (Default|UnbindEventfromResOut (Variables|Private|GetLast%20Pawn) _outputdelegate))
          (:then_1
            _))
        (else
          _))
      (bind _asplayer_pawn_athena_2 (Utilities|Casting|CastToPlayerPawn_Athena (CustomCharacter|GetAssociatedPlayerPawn)))
      (bind _output_get (Variables|Private|SetLast%20Pawn _asplayer_pawn_athena_2))
      ;;[1 1 1] This is binding to the Event in the player pawn that is called by Athena player pawn by GA_Athena_HidingProp
      (Default|BindEventtoCCPMIsHiddeninProp _output_get (EventDispatchers|CreateEvent :!event isHiddingInProp))
      (Default|BindEventtoResIn (Variables|Private|GetLast%20Pawn) _outputdelegate_2)
      (Default|BindEventtoResOut (Variables|Private|GetLast%20Pawn) _outputdelegate))))

(event Custom|Attachment)

(event Custom|isHiddingInProp
  (params (isHidden :type Boolean))
  ;[1 1 1] The origin is in GA_Athena_HidingProp_Hide which calls isHiddingInProp event from Player Pawn Athena. isHiddingInProp is bound to a CCPM_isHiddenInProp dispatcher. CCPM_isHiddenInProp dispatcher is bound to isHiddingInProp
  (if isHidden
    (CallFunction|EnteringHidingProp)
    (else
      (CallFunction|ExitingHidingProp))))

(event EventOnEndSkydiving
  ;[0 0 0 0.5 40] Toggle visibility, this function hides the dynamically created components. As you add components to the children of this base class, they are hidden by this graph.
  (|CallParentFunction)
  (if (Variables|CustomCharacterPartModifierBase|GetDisable%20Particles%20during%20Skydive)
    (Development|PrintString "End Skydiving" (Variables|Debug|GetEnable%20Debug%20Printing) (Variables|Debug|GetEnable%20Debug%20Printing) (Variables|Debug|GetEvent%20Log%20Color))
    (CustomCharacter|SetChildParticleComponentsVisible true)))

(event EventOnTargetingOrFirstPersonCameraChanged)

(event Custom|ExitVehicle
  ;[0 0 0 0.5 40] Toggle visibility, this function hides the dynamically created components. As you add components to the children of this base class, they are hidden by this graph.
  (if (Variables|CustomCharacterPartModifierBase|GetDisable%20Particles%20in%20Vehicle)
    (Development|PrintString "Exit Vehicle" (Variables|Debug|GetEnable%20Debug%20Printing) (Variables|Debug|GetEnable%20Debug%20Printing) (Variables|Debug|GetEvent%20Log%20Color))
    (CustomCharacter|SetChildParticleComponentsVisible true)))

(event Custom|Open
  ;[0 0 0 0.5 40] Toggle visibility, this function hides the dynamically created components. As you add components to the children of this base class, they are hidden by this graph.
  (Development|PrintString "Open" (Variables|Debug|GetEnable%20Debug%20Printing) (Variables|Debug|GetEnable%20Debug%20Printing) (Variables|Debug|GetEvent%20Log%20Color))
  (CustomCharacter|SetChildMeshComponentsHidden false))

(event Custom|ExitingHidingProp
  ;[0 0 0 0.5 40] Toggle visibility, this function hides the dynamically created components. As you add components to the children of this base class, they are hidden by this graph.
  (Development|PrintString "Exit Hiding Prop" (Variables|Debug|GetEnable%20Debug%20Printing) (Variables|Debug|GetEnable%20Debug%20Printing) (Variables|Debug|GetEvent%20Log%20Color))
  (Utilities|FlowControl|Sequence
    (:then_0
      (CustomCharacter|SetChildParticleComponentsVisible true))
    (:then_1
      (CustomCharacter|SetChildMeshComponentsHidden false))))

(event EventOnBeginSkydiving
  (bind _disable_particles_during_skydive (Variables|CustomCharacterPartModifierBase|GetDisable%20Particles%20during%20Skydive))
  ;[0 0 0 0.5 40] Toggle visibility, this function hides the dynamically created components. As you add components to the children of this base class, they are hidden by this graph.
  (|CallParentFunction)
  (Development|PrintString "Skydiving" false false "(R=0.000000,G=1.000000,B=0.425653,A=1.000000)" 15.0)
  (Development|PrintString (Utilities|String|Append "Disable Particles During Skydive: " (Utilities|String|ToString(Boolean) _disable_particles_during_skydive)) false false "(R=0.000000,G=0.000000,B=0.000000,A=1.000000)" 1e+01)
  (if _disable_particles_during_skydive
    (Development|PrintString "Disabling Particles during Skydive" false false (Variables|Debug|GetEvent%20Log%20Color) 2e+01)
    (CustomCharacter|SetChildParticleComponentsVisible false)
    (else
      (Development|PrintString "Enabling Particles during Skydive" false false "(R=1.000000,G=0.000000,B=0.232080,A=1.000000)" 2e+01))))

(event Custom|EnterVehicle
  ;[0 0 0 0.5 40] Toggle visibility, this function hides the dynamically created components. As you add components to the children of this base class, they are hidden by this graph.
  (if (Variables|CustomCharacterPartModifierBase|GetDisable%20Particles%20in%20Vehicle)
    (Development|PrintString "Entering Vehicle" false false (Variables|Debug|GetEvent%20Log%20Color))
    (CustomCharacter|SetChildParticleComponentsVisible false)))

(event Custom|Close
  ;[0 0 0 0.5 40] Toggle visibility, this function hides the dynamically created components. As you add components to the children of this base class, they are hidden by this graph.
  (CustomCharacter|SetChildMeshComponentsHidden true))

(event Custom|EnteringHidingProp
  ;[0 0 0 0.5 40] Toggle visibility, this function hides the dynamically created components. As you add components to the children of this base class, they are hidden by this graph.
  (Development|PrintString "Entering Prop" false false (Variables|Debug|GetEvent%20Log%20Color))
  (Utilities|FlowControl|Sequence
    (:then_0
      (CustomCharacter|SetChildMeshComponentsHidden true))
    (:then_1
      (CustomCharacter|SetChildParticleComponentsVisible false))))

(event Custom|PlayerMovementTick
  (params (DeltaSeconds :type Float%20%28single-precision%29) (OldLocation :type Vector) (OldVelocity :type Vector))
  (bind _asplayer_pawn_athena (Utilities|Casting|CastToPlayerPawn_Athena (CustomCharacter|GetAssociatedPlayerPawn)))
  (bind _output_get (Variables|Private|SetbIsDriving (or (Vehicle|IsDrivingVehicle _asplayer_pawn_athena) (Vehicle|IsinVehicle _asplayer_pawn_athena))))
  (if _output_get
    (Utilities|FlowControl|DoOnce true
      (:Completed
        (CallFunction|EnterVehicle)
        (CallFunction|ResetExitVehicle)))
    (else
      (Utilities|FlowControl|DoOnce false
        (:Completed
          (CallFunction|ExitVehicle)
          (CallFunction|ResetEnterVehicle))))))

(event Custom|Reset%20Enter%20Vehicle
  (Utilities|FlowControl|DoOnce true
    (:Completed
      (CallFunction|EnterVehicle)
      (CallFunction|ResetExitVehicle))))

(event Custom|Reset%20Exit%20Vehicle
  (Utilities|FlowControl|DoOnce false
    (:Completed
      (CallFunction|ExitVehicle)
      (CallFunction|ResetEnterVehicle))))

(fn attachParticleSystem
  (params (Particle%20System%20Component :type Cascade%20Particle%20System%20Component%20Object%20Reference) (Socket%20Name :type Name))
  (Transformation|AttachComponentToComponent Particle%20System%20Component (Variables|Private|GetCharacter%20Part%20Type) Socket%20Name "SnapToTarget" "SnapToTarget" "KeepRelative" true))

(fn ConstructionScript
  (bind _childcomponenttag (Variables|CustomCharacterPartModifier|GetChildComponentTag))
  (Utilities|FlowControl|Sequence
    (:then_0
      (Variables|Debug|SetHidding%20prefixes (Utilities|Array|MakeArray "Hidding" "Unhiding"))
      ;[1 1 1] Sets the name of the object in the component tag for visibility toggling.
      ;;[1 1 1] Get display name of this object. This is a global variable that is used in the hiding functions.
      (Variables|CustomCharacterPartModifier|SetChildComponentTag (Utilities|String|StringToName (Utilities|GetDisplayName self)))
      (for _array_element (Components|GetChildrenComponents (Variables|Default|GetDefaultSceneRoot) true)
        (Development|PrintString (Utilities|String|Append "Tagging" (Utilities|GetDisplayName _array_element) "with :" (Utilities|String|ToString(Name) _childcomponenttag)) (Variables|Debug|GetEnable%20Debug%20Printing) (Variables|Debug|GetEnable%20Debug%20Printing) "(R=1.000000,G=0.000000,B=0.787428,A=1.000000)" 1e+01)
        (Variables|/Script/Engine.ActorComponent|SetComponentTags (Utilities|Array|MakeArray _childcomponenttag) _array_element)))))
